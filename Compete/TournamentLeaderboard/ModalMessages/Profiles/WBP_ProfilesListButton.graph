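# List entry button for the profiles list in tournament leaderboard modals, showing one player's name, built on WBP_LeaderboardListEntry_Base.
# Setting the list item stores its VM_CompetePlayerProfile and the nested tournament profile view model.
# UpdateName formats NickName as rich text with a country flag via BP_FlagsDecorator, or name only when the flag is "fortnite".
# Selection and deselection play a text-material color animation forward or in reverse, timed by UIKit's Selecting block timing.
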WidgetBlueprint /Compete/TournamentLeaderboard/ModalMessages/Profiles/WBP_ProfilesListButton
animations [Anim_TransitionSelected {
  binding_count=1
  length_seconds=1
  object=&self:Anim_TransitionSelected
  sequence {bindings [NickName {
     id=#1
     possessed_class=CommonRichTextBlock
     tracks [{
       type=MovieSceneWidgetMaterialTrack
       display_name=Default Text Style Override.Font.FontMaterial
       sections [{
         type=MovieSceneParameterSection
         end_frame=20
         start_frame=0
         channels [
          Trans_Focused {type=MovieSceneScriptingFloatChannel, const [0,2]}
          Trans_Hovered {type=MovieSceneScriptingFloatChannel, const [0,2]}
          Trans_Pressed {type=MovieSceneScriptingFloatChannel, const [0,2]}
          Color_IdleR {type=MovieSceneScriptingFloatChannel, keys [[0,1],[20,0]]}
          Color_IdleG {type=MovieSceneScriptingFloatChannel, keys [[0,1],[20,0]]}
          Color_IdleB {type=MovieSceneScriptingFloatChannel, keys [[0,1],[20,0]]}
          Color_IdleA {type=MovieSceneScriptingFloatChannel, const [1,2]}
          Color_HoveredR {type=MovieSceneScriptingFloatChannel, keys [[0,0.0103],[20,0]]}
          Color_HoveredG {type=MovieSceneScriptingFloatChannel, keys [[0,0.0103],[20,0]]}
          Color_HoveredB {type=MovieSceneScriptingFloatChannel, keys [[0,0.0103],[20,0]]}
          Color_HoveredA {type=MovieSceneScriptingFloatChannel, const [1,2]}
          Color_FocusedR {type=MovieSceneScriptingFloatChannel, keys [[0,0.0103],[20,0]]}
          Color_FocusedG {type=MovieSceneScriptingFloatChannel, keys [[0,0.0103],[20,0]]}
          Color_FocusedB {type=MovieSceneScriptingFloatChannel, keys [[0,0.0103],[20,0]]}
          Color_FocusedA {type=MovieSceneScriptingFloatChannel, const [1,2]}
          Color_PressedR {type=MovieSceneScriptingFloatChannel, keys [[0,0.0103],[20,0]]}
          Color_PressedG {type=MovieSceneScriptingFloatChannel, keys [[0,0.0103],[20,0]]}
          Color_PressedB {type=MovieSceneScriptingFloatChannel, keys [[0,0.0103],[20,0]]}
          Color_PressedA {type=MovieSceneScriptingFloatChannel, const [1,2]}
         ]
        }]
      }]
    }]}
 }]
graphs [
 EventGraph {nodes [
   n1 = Event OnListItemObjectSet()
   > n2 = Cast To VM_CompetePlayerProfile!(execute, Object=@n1.ListItemObject) -> (then, CastFailed, AsVM Compete Player Profile:VM Compete Player Profile Object Reference) {is=title K2Node_DynamicCast}
   > n3 = Set VM_CompetePlayerProfile(VM_CompetePlayerProfile:VM Compete Player Profile Object Reference=@n2) -> (Output_Get:VM Compete Player Profile Object Reference)
   > Set FortTournamentPlayerProfileVM(FortTournamentPlayerProfileVM:Tournament Player Profile VM Object Reference=Get VMProfile!(self:VM Compete Player Profile Object Reference=@n3) -> (VMProfile:Tournament Player Profile VM Object Reference)) -> (Output_Get:Tournament Player Profile VM Object Reference)
   Event On Selected()
   > Parent: On Selected() {is=title K2Node_CallParentFunction}
   n4 = Get NickName() -> (NickName:Common Rich Text Block Object Reference)
   > SetDefaultMaterial!(
    execute
    self:Rich Text Block Object Reference=@n4
    InMaterial:Material Interface Object Reference=/FortUILibrary/UIKit/Buttons/MaterialInstances/SelectableButtons/MI_UIKit_SelectableText_Quiet
   ) -> (then)
   > GetDefaultDynamicMaterial!(execute, self:Rich Text Block Object Reference=@n4) -> (then, ReturnValue:Material Instance Dynamic Object Reference)
   n5 = float / float(
    A=1
    B=Break UIKit Block Timing!(UIKitBlockTiming:"UIKit Block Timing Structure (by ref)"=GetBlockTimings!() -> (BlockTiming:UIKit Block Timing Structure)) -> (Selecting:float=~) {is=title K2Node_BreakStruct}
   )
   > PlayAnimationForward(InAnimation=Get Anim_TransitionSelected() -> (Anim_TransitionSelected:Widget Animation Object Reference), PlaybackSpeed=@n5)
   Event On Deselected()
   > Parent: On Deselected() {is=title K2Node_CallParentFunction}
   > PlayAnimationReverse(InAnimation=Get Anim_TransitionSelected(), PlaybackSpeed=@n5)
  ]}
 UpdateName {nodes [
   UpdateName!() -> (then, Name:String) {is=title K2Node_FunctionEntry}
   n1 = Get FortTournamentPlayerProfileVM() -> (FortTournamentPlayerProfileVM:Tournament Player Profile VM Object Reference)
   n2 = Get FlagName!(self:Tournament Player Profile VM Object Reference=@n1) -> (FlagName:String)
   n3 = Get DisplayName!(self:Tournament Player Profile VM Object Reference=@n1) -> (DisplayName:String)
   > SetText(
    self=Get NickName()
    InText=Select(
     Option 0:"Text (by ref)"=Format Text!(Format="LOCTEXT(\"<img id=\\\"{Country}\\\" size=\\\"18,18\\\" height=\\\"25\\\"/> {Nickname}\")", Country:String=@n2, Nickname:String=@n3) -> (Result)
     Option 1:"Text (by ref)"=To Text!(Value:String=@n3) -> (ReturnValue:Text) {is=title K2Node_GenericToText}
     Index="Equal Exactly (String)"(A=@n2, B=fortnite)
    ) -> (ReturnValue:"Text (by ref)")
   )
  ]}
]
variables [
 BackgroundMaterial {type=MaterialInstanceDynamic}
 MatchResultValuesFont {
  type=SlateFontInfo
  default {
   bForceMonospaced=false
   bMaterialIsStencil=false
   fontMaterial=&MI_UIKit_SelectableText_Quiet
   fontObject=&HeadingNow
   letterSpacing=0
   monospacedWidth=1
   size=18
   skewAmount=0
   typefaceFontName=75 Medium
   outlineSettings {
    bApplyOutlineToDropShadows=false
    bMiteredCorners=false
    bSeparateFillAlpha=false
    outlineMaterial=null
    outlineSize=0
    outlineColor=<rgba 0 0 0 1>
   }
  }
 }
]
widgets {
 info {
  inheritedWidgetCount=0
  namedSlotCount=1
  parentClass=/Compete/TournamentLeaderboard/ModalMessages/WBP_LeaderboardListEntry_Base.WBP_LeaderboardListEntry_Base_C
  rootWidgetClass=null
  widgetCount=2
 }
 widgets [
  Main {widgetClassPath=&GridPanel, bIsVariable=true, widget=Main, properties {columnFill [1], rowFill [1]}}
  NickName {
   widgetClassPath=&CommonRichTextBlock
   bIsVariable=true
   parent=Main
   widget=NickName
   properties {
    autoWrapText=true
    bIsScrollingEnabled=false
    bOverrideDefaultStyle=true
    mobileTextBlockScale=1.14
    text="WWWWWWWWWWWWWWWW "
    decoratorClasses [/Compete/DataTables/BP_FlagsDecorator.BP_FlagsDecorator_C]
    defaultTextStyleOverride {
     overflowPolicy=Clip
     transformPolicy=None
     colorAndOpacity {colorUseRule=UseColor_Specified, specifiedColor=<rgba 0.5765 0.5765 0.5765 0.8>}
     font {
      bForceMonospaced=false
      bMaterialIsStencil=false
      fontMaterial=&MI_UIKit_SelectableText_Quiet
      fontObject=&HeadingNow
      letterSpacing=0
      monospacedWidth=1
      size=15.75
      skewAmount=0
      typefaceFontName=75 Medium
      outlineSettings {
       bApplyOutlineToDropShadows=false
       bMiteredCorners=false
       bSeparateFillAlpha=false
       outlineMaterial=null
       outlineSize=0
       outlineColor=<rgba 0 0 0 1>
      }
     }
     highlightColor {colorUseRule=UseColor_Specified, specifiedColor=<rgba 0 0 0 1>}
     highlightShape {
      bIsDynamicallyLoaded=false
      drawAs=Image
      imageType=NoImage
      mirroring=NoMirror
      resourceName=None
      resourceObject=null
      tiling=NoTile
      imageSize=<32 32>
      margin=<ltrb 0 0 0 0>
      outlineSettings {
       bUseBrushTransparency=false
       roundingType=HalfHeightRadius
       width=0
       color {colorUseRule=UseColor_Specified, specifiedColor=<rgba 0 0 0 0>}
       cornerRadii=<0 0 0 0>
      }
      tintColor {colorUseRule=UseColor_Specified, specifiedColor=<rgba 1 1 1 1>}
      uVRegion {bIsValid=false, max=<0 0>, min=<0 0>}
     }
     selectedBackgroundColor {colorUseRule=UseColor_Foreground, specifiedColor=<rgba 1 0 1 1>}
     shadowColorAndOpacity=<rgba 0 0 0 1>
     shadowOffset=<0 0>
     strikeBrush {
      bIsDynamicallyLoaded=false
      drawAs=Image
      imageType=NoImage
      mirroring=NoMirror
      resourceName=None
      resourceObject=null
      tiling=NoTile
      imageSize=<32 32>
      margin=<ltrb 0 0 0 0>
      outlineSettings {
       bUseBrushTransparency=false
       roundingType=HalfHeightRadius
       width=0
       color {colorUseRule=UseColor_Specified, specifiedColor=<rgba 0 0 0 0>}
       cornerRadii=<0 0 0 0>
      }
      tintColor {colorUseRule=UseColor_Specified, specifiedColor=<rgba 1 1 1 1>}
      uVRegion {bIsValid=false, max=<0 0>, min=<0 0>}
     }
     underlineBrush {
      bIsDynamicallyLoaded=false
      drawAs=Image
      imageType=NoImage
      mirroring=NoMirror
      resourceName=None
      resourceObject=null
      tiling=NoTile
      imageSize=<32 32>
      margin=<ltrb 0 0 0 0>
      outlineSettings {
       bUseBrushTransparency=false
       roundingType=HalfHeightRadius
       width=0
       color {colorUseRule=UseColor_Specified, specifiedColor=<rgba 0 0 0 0>}
       cornerRadii=<0 0 0 0>
      }
      tintColor {colorUseRule=UseColor_Specified, specifiedColor=<rgba 1 1 1 1>}
      uVRegion {bIsValid=false, max=<0 0>, min=<0 0>}
     }
    }
   }
   slot {class=GridSlot, properties {horizontalAlignment=HAlign_Left, layer=1, verticalAlignment=VAlign_Center, padding=<ltrb 11 0 11 0>}}
  }
 ]
}
